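# Post-process style animation Blueprint for the Crown face accessory mesh of the male medium WeaveHarbor character, copying its pose from the attached parent.
# The pose chain adjusts the hair corrective bone (defomr_hair_corr), then the head bone with a fixed scale tweak.
# Each update sets LOD flags and maps head rotation relative to spine_05 on the customization base mesh into eased corrective location and scale variables, with no visible AnimGraph link.
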AnimBlueprint /BRCosmetics/Characters/Player/Male/Medium/Bodies/M_MED_WeaveHarbor_Crown/Meshes/Parts/M_MED_WeaveHarbor_Crown_FaceAcc_AnimBP
target_skeleton=&/M_MED_WeaveHarbor_Crown_FaceAcc_Skeleton
graphs [
 AnimGraph {nodes [
   "Transform (Modify) Bone - Bone: defomr_hair_corr"!(ComponentPose, bAlphaBoolEnabled:bool=True, Translation="0,-0.5,0.4", Rotation="0,0,0", Scale="1,1,1") -> (Pose) {
    alphaInputType=Bool
    scaleMode=BMM_Additive
    translationMode=BMM_Additive
    translationSpace=BCS_BoneSpace
    alphaBoolBlend {blendOption=Linear}
    boneToModify {boneName=defomr_hair_corr}
    translation {y=-0.5, z=0.4}
   }
   "Transform (Modify) Bone - Bone: defomr_hair_corr"(bAlphaBoolEnabled=True, Translation="0,-0.5,0.4", Rotation="0,0,0", Scale="1,1,1") {
    alphaInputType=Bool
    scaleMode=BMM_Additive
    translationMode=BMM_Additive
    translationSpace=BCS_BoneSpace
    alphaBoolBlend {blendOption=Linear}
    boneToModify {boneName=defomr_hair_corr}
    translation {y=-0.5, z=0.4}
   }
   "Transform (Modify) Bone - Bone: defomr_hair_corr"(bAlphaBoolEnabled=True, Translation="0,0,1", Rotation="0,0,0", Scale="1,1,1") {
    alphaInputType=Bool
    scaleMode=BMM_Additive
    translationMode=BMM_Additive
    alphaBoolBlend {blendOption=Linear}
    boneToModify {boneName=defomr_hair_corr}
    translation {z=1}
   }
   n1 = "Transform (Modify) Bone - Bone: defomr_hair_corr"!(
    ComponentPose=Local To Component(LocalPose=Copy Pose From Mesh() {bUseAttachedParent=true})
    Alpha:float=1
    Translation="0,0,0"
    Rotation="0,0,0"
    Scale="1,1,1"
   ) -> (Pose) {
    scaleMode=BMM_Additive
    scaleSpace=BCS_BoneSpace
    translationMode=BMM_Additive
    translationSpace=BCS_BoneSpace
    alphaBoolBlend {blendOption=Linear}
    boneToModify {boneName=defomr_hair_corr}
   }
   n2 = "Transform (Modify) Bone - Bone: head"!(ComponentPose=@n1, bAlphaBoolEnabled:bool=True, Translation="0,0,0", Rotation="0,0,0", Scale="1,1,1") -> (Pose) {
    alphaInputType=Bool
    scaleMode=BMM_Additive
    scaleSpace=BCS_ParentBoneSpace
    translationMode=BMM_Additive
    translationSpace=BCS_ParentBoneSpace
    alphaBoolBlend {blendOption=Linear}
    boneToModify {boneName=head}
   }
   n3 = "Transform (Modify) Bone - Bone: head"(ComponentPose=@n2, bAlphaBoolEnabled=True, Translation="0,0,0", Rotation="0,0,0", Scale="1,1.1,1.06") {
    alphaInputType=Bool
    scaleMode=BMM_Additive
    scaleSpace=BCS_BoneSpace
    translationMode=BMM_Additive
    translationSpace=BCS_ParentBoneSpace
    alphaBoolBlend {blendOption=Linear}
    boneToModify {boneName=head}
    scale {y=1.1, z=1.06}
   }
   Output Pose(Result=Component To Local(ComponentPose=@n3)) {layerGroup=None}
  ]}
 EventGraph {nodes [
   Event BlueprintUpdateAnimation()
   n1 = Get CurrentLOD()
   > Set IsLOD2(IsLOD2:bool="integer >= integer"(A=@n1, B=2)) -> (Output_Get:bool)
   > Set isLOD1(isLOD1:bool="Equal (Integer)"(A=@n1, B=1)) -> (Output_Get:bool)
   n2 = Ease!(
    Function:EEasingFunc Enum=Linear
    Alpha:double=GetAlphaFromReferenceTransformDelta(
     SkeletalMeshComponent=Get Custom Base Mesh() -> (Custom Base Mesh:Skeletal Mesh Component Object Reference)
     BoneName=head
     BaseName=spine_05
     InputRange=(X=0,Y=-60)
     OutputRange=(X=0.2,Y=1)
     TwistAxis="1,0,0"
    )
    A:Transform=MakeTransform(Rotation="0, 0, 0")
    B:Transform=MakeTransform(Location="0,-3,0", Rotation="0, 0, 0", Scale="1.5,1.2,1")
   ) -> (Result_Location:Vector, Result_Rotation:Rotator, Result_Scale:Vector) {is=title K2Node_EaseFunction}
   > Set Result Location(Result Location:Vector=@n2) -> (Output_Get:Vector)
   > Set Result Scale(Result Scale:Vector=@n2.Result_Scale) -> (Output_Get:Vector)
   n3 = Event OnCharacterCustomizationLoaded()
   > Set Custom Base Mesh(Custom Base Mesh:Skeletal Mesh Component Object Reference=@n3.CustomBaseMesh) -> (Output_Get:Skeletal Mesh Component Object Reference)
  ]}
]
variables [
 Custom Base Mesh {type=SkeletalMeshComponent}
 isLOD1 {type=boolean}
 IsLOD2 {type=boolean}
 Result Location {type=Vector}
 Result Scale {type=Vector}
]
